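# Niagara ribbon trail for the First Class pickaxe, with one emitter that follows skeletal mesh bones or sockets through shared anim-trail modules.
# Short-lived particles feed three ribbon renderers, with time-of-day tints, age-driven U coordinates and alpha fade.
# A camera-vector dot product against the owner's Y axis drives a dynamic material parameter.
# When execution becomes inactive, a Timeline blends in alpha curves indexed by execution order to fade the ribbon.
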NiagaraSystem /BRCosmetics/Weapons/FORT_Melee/Pickaxe_FirstClass/FX/NS_Pickaxe_FirstClass_Trail
dependencies {
 usedDataInterfaces [&NiagaraDataInterfaceCurve, &NiagaraDataInterfaceCamera, &NiagaraDataInterfaceSkeletalMesh]
 usedDynamicInputs []
 usedModules [
  &self:SystemSpawnScript.SystemScriptSource.SystemScriptGraph.NiagaraNodeAssignment_0.SetVariables_#1
  &SystemState
 ]
}
emitters {MainTrail {
  emitterSpawnScript [&FN_RibbonIDOffsetUtility()]
  emitterUpdateScript [
   &EmitterState()
   &FN_SpawnAnimTrail()
  ]
  particleSpawnScript [
   &FN_InitializeAnimTrailBoneIndecies(Second Bone or Socket Index=0)
   &FN_InitializeAnimTrail(
    Color=&FNTimeOfDayColor(Night Color=<rgba 1 1 1 0.25>, Morning Color=<rgba 1 1 1 0.3>, Day Color=<rgba 1 1 1 0.4>, Evening Color=<rgba 1 1 1 0.3>)
    Lifetime=0.125
    Position Offset=&TransformVector(Vector=<0 -15 0>, Destination Space=::Simulation)
    SkeletalMesh=$System.Skeletal Mesh
   )
   /SetVariables(Particles.RibbonFacing=$Engine.Owner.SystemZAxis, Particles.RibbonWidth=36.7565) {isSetParameters=true}
   /SetVariables(Particles.HoopColor:LinearColor=&FNTimeOfDayColor(
     Night Color=<rgba 0.099 0.0062 0.0072 0.5>
     Morning Color=<rgba 0.1927 0.012 0.0139 0.5>
     Day Color=<rgba 0.2604 0.0163 0.0189 0.5>
     Evening Color=<rgba 0.1979 0.0124 0.0143 0.5>
    )) {isSetParameters=true}
  ]
  particleUpdateScript [
   &ParticleState()
   &Timeline(
    Play=&Boolean_NotOperation(Bool=/Game/Effects/Niagara/DynamicInputs/State/FN_IsExecutionStateActive())
    Play Rate=&Divide_Float(A=1, B=0.08)
    Rewind Mode:ENiagara_TimelineMode=::NewEnumerator1
   )
   &FN_UpdateAnimTrail(SkeletalMesh=$System.Skeletal Mesh) {module=FN_UpdateAnimTrail001}
   /SetVariables(Particles.RibbonU0Override=$Particles.NormalizedAge) {isSetParameters=true}
   &ScaleColor(
    Scale Alpha=&FloatFromCurve(FloatCurve=DataInterface() {propertyValues {Curve {keys [<at 0 1 RCIM_Cubic>, {arriveTangent=-1.6057, tangentMode=RCTM_User, time=1, value=0}]}}})
    ScaleA=true
    ScaleRGB=true
   ) {module=ScaleColor001}
   &DynamicMaterialParameters(
    Index 0 Param 1=&Lerp_Float(
     A=0
     B=1
     Alpha=&FloatFromCurve(
      FloatCurve=DataInterface() {propertyValues {Curve {keys [<at 0.5 0>, {tangentMode=RCTM_User, time=1, value=1}]}}}
      CurveIndex=&Abs_Float(Float:NiagaraFloat=&DotProduct(
        Vector A=&ReturnCameraVector(Camera Vectors=::NewEnumerator1, Coordinate Space=::Simulation)
        Vector B=$Engine.Owner.SystemYAxis
       ))
     )
    )
    Module.Index0Param0Write=true
    Module.Index0Param1Write=true
    Module.Index0Param2Write=true
    Module.Index0Param3Write=true
   ) {module=DynamicMaterialParameters001}
   /SetVariables(Particles.Linear Color=&MultiplyLinearColor(
     Linear Color=<rgba 0 0 0 0.5>
     Scale Factor=&MakeFloat4FromVectorAndFloat(
      "Vector (RGB)"=<1 1 1>
      "Float (Alpha)"=&Lerp_Float(
       A=1
       B=&FloatFromCurve(
        FloatCurve=DataInterface() {propertyValues {Curve {keys [
            {
             arriveTangentWeight=0.4932
             interpMode=RCIM_Cubic
             leaveTangentWeight=0.4932
             tangentMode=RCTM_User
             tangentWeightMode=RCTWM_WeightedBoth
             time=0
             value=0
            }
            {
             arriveTangent=15.8665
             arriveTangentWeight=0.8047
             interpMode=RCIM_Cubic
             leaveTangent=-2.9657
             leaveTangentWeight=0.5216
             tangentMode=RCTM_Break
             tangentWeightMode=RCTWM_WeightedBoth
             time=0.5
             value=1
            }
            {interpMode=RCIM_Cubic, leaveTangent=-0.5789, time=1, value=0}
           ]}}}
        CurveIndex=&ReturnNormalizedExecIndex()
       )
       Alpha=$Particles.Timeline.NormalizedTime
      )
     )
    )) {isSetParameters=true}
   &ScaleColor(
    Scale Alpha=&Lerp_Float(
     A=1
     B=&FloatFromCurve(
      FloatCurve=DataInterface() {propertyValues {Curve {keys [
          {leaveTangentWeight=0.1667, tangentMode=RCTM_User, time=0, value=1}
          {
           arriveTangent=-6.6819
           arriveTangentWeight=1.1261
           interpMode=RCIM_Cubic
           leaveTangent=-6.9931
           leaveTangentWeight=0.4148
           tangentMode=RCTM_Break
           tangentWeightMode=RCTWM_WeightedBoth
           time=0.4
           value=1
          }
          {
           arriveTangentWeight=0.3965
           interpMode=RCIM_Cubic
           leaveTangentWeight=0.3965
           tangentMode=RCTM_User
           tangentWeightMode=RCTWM_WeightedBoth
           time=1
           value=0
          }
         ]}}}
      CurveIndex=&ReturnNormalizedExecIndex()
     )
     Alpha=$Particles.Timeline.NormalizedTime
    )
    ScaleA=true
    ScaleRGB=true
   )
  ]
  rendererClasses [&NiagaraRibbonRendererProperties]
  renderers [
   {rendererClass=&NiagaraRibbonRendererProperties}
   {rendererClass=&NiagaraRibbonRendererProperties, rendererIndex=1}
   {rendererClass=&NiagaraRibbonRendererProperties, rendererIndex=2}
  ]
 }}
summary {
 systemName=NS_Pickaxe_FirstClass_Trail
 emitters [{emitterName=MainTrail, rendererClasses [&NiagaraRibbonRendererProperties]}]
 userVariables [
  User.bIsFrontEnd {type=NiagaraBool}
  User.bIsFrontEndPreview {type=NiagaraBool}
  User.RibbonIDOffset {type=NiagaraInt32}
 ]
}
