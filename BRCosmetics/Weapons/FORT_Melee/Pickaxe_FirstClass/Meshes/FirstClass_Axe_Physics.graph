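# Physics asset for the FirstClass pickaxe mesh, a two-handed melee cosmetic with a simulated dangling net piece.
# A large capsule covers the handle, and three small capsules form the net_base bone chain.
# Each net segment is constrained to the one before it, starting from the handle, with narrow swing limits and almost no twist, so the chain sways slightly.
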PhysicsAsset /BRCosmetics/Weapons/FORT_Melee/Pickaxe_FirstClass/Meshes/FirstClass_Axe_Physics
bodies {
 Melee_TwoHanded_Handle [{
   shapeType=Capsule
   length=76.6912
   radius=6.8092
   shapeName=Melee_TwoHanded_Handle_capsule
   center=<-0.1191 -0.7974 23.062>
   rotation=<rot 3.0196 -90.0258 -0.2959>
  }]
 net_base_1 [{shapeType=Capsule, length=3, radius=3, shapeName=net_base_1_capsule, center=<2 0 0>, rotation=<rot 90 0 0>}]
 net_base_2 [{shapeType=Capsule, length=3, radius=3, shapeName=net_base_2_capsule, center=<4 0 0>, rotation=<rot 90 0 0>}]
 net_base_3 [{shapeType=Capsule, length=3, radius=3, shapeName=net_base_3_capsule, center=<2.5 0 0>, rotation=<rot 90 0 0>}]
}
constraints [
 {bone1Name=net_base_1, bone2Name=Melee_TwoHanded_Handle, swing1LimitDegrees=10, swing2LimitDegrees=10, twistLimitDegrees=1}
 {bone1Name=net_base_2, bone2Name=net_base_1, swing1LimitDegrees=22, swing2LimitDegrees=22, twistLimitDegrees=1}
 {bone1Name=net_base_3, bone2Name=net_base_2, swing1LimitDegrees=15, swing2LimitDegrees=15, twistLimitDegrees=1}
]
