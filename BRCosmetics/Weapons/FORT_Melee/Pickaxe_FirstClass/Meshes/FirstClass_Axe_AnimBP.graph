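# Animation Blueprint for the FirstClass pickaxe that adds secondary physics motion to the weapon and its net.
# A RigidBody node simulates the physics asset relative to b_AR_Root, with gravity updated each frame by a deprecated override function.
# Weighted constraints then drive sixteen net tip base bones from three net base bones and the handle.
# Finally, per-tip AnimDynamics nodes add angle-limited, spring-driven jiggle with their own fixed gravity.
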AnimBlueprint /BRCosmetics/Weapons/FORT_Melee/Pickaxe_FirstClass/Meshes/FirstClass_Axe_AnimBP
target_skeleton=&/FirstClass_Axe_Skeleton
graphs [
 AnimGraph {nodes [
   AnimGraphNode_SubInput_0 = Input Pose()
   n1 = RigidBody!(
    ComponentPose=Local To Component(LocalPose=@AnimGraphNode_SubInput_0)
    Alpha
    OverrideWorldGravity:Vector=Get Gravity Override()
    ExternalForce
   ) -> (Pose) {
    bOverrideWorldGravity=true
    overridePhysicsAsset=&/FirstClass_Axe_Physics
    simulationSpace=BaseBoneSpace
    alphaBoolBlend {blendOption=Linear}
    baseBoneRef {boneName=b_AR_Root}
    componentAppliedLinearAccClamp=<3000 3000 3000>
    componentLinearAccScale=<0.35 0.35 0.35>
    componentLinearVelScale=<0.35 0.35 0.35>
    simSpaceSettings {
     maxAngularAcceleration=1000
     maxAngularVelocity=700
     maxLinearAcceleration=500
     maxLinearVelocity=5
     velocityScaleZ=0.7
     worldAlpha=0.45
     externalLinearDragV=<0.2 0.2>
    }
   }
   n2 = "Constraint - net_tip_1_base (PPPP)"!(
    ComponentPose=@n1
    Alpha=1
    ConstraintWeights_0=0.6725
    ConstraintWeights_1:float=0.2724
    ConstraintWeights_2:float=0.0546
    ConstraintWeights_3:float=0.0005
   ) -> (Pose) {
    alphaBoolBlend {blendOption=Linear}
    boneToModify {boneName=net_tip_1_base}
    constraintSetup [
     {targetBone {boneName=net_base_2}}
     {targetBone {boneName=net_base_3}}
     {targetBone {boneName=net_base_1}}
     {targetBone {boneName=Melee_TwoHanded_Handle}}
    ]
    constraintWeights [1,1,1,1]
   }
   n3 = "Constraint - net_tip_2_base (PPPP)"!(
    ComponentPose=@n2
    Alpha=1
    ConstraintWeights_0=0.6562
    ConstraintWeights_1:float=0.2969
    ConstraintWeights_2:float=0.0465
    ConstraintWeights_3:float=0.0004
   ) -> (Pose) {
    alphaBoolBlend {blendOption=Linear}
    boneToModify {boneName=net_tip_2_base}
    constraintSetup [
     {targetBone {boneName=net_base_2}}
     {targetBone {boneName=net_base_3}}
     {targetBone {boneName=net_base_1}}
     {targetBone {boneName=Melee_TwoHanded_Handle}}
    ]
    constraintWeights [1,1,1,1]
   }
   n4 = "Constraint - net_tip_3_base (PPPP)"!(
    ComponentPose=@n3
    Alpha=1
    ConstraintWeights_0=0.6387
    ConstraintWeights_1:float=0.3235
    ConstraintWeights_2:float=0.0376
    ConstraintWeights_3:float=0.0003
   ) -> (Pose) {
    alphaBoolBlend {blendOption=Linear}
    boneToModify {boneName=net_tip_3_base}
    constraintSetup [
     {targetBone {boneName=net_base_2}}
     {targetBone {boneName=net_base_3}}
     {targetBone {boneName=net_base_1}}
     {targetBone {boneName=Melee_TwoHanded_Handle}}
    ]
    constraintWeights [1,1,1,1]
   }
   n5 = "Constraint - net_tip_4_base (PPPP)"!(
    ComponentPose=@n4
    Alpha=1
    ConstraintWeights_0=0.6775
    ConstraintWeights_1:float=0.2648
    ConstraintWeights_2:float=0.0572
    ConstraintWeights_3:float=0.0005
   ) -> (Pose) {
    alphaBoolBlend {blendOption=Linear}
    boneToModify {boneName=net_tip_4_base}
    constraintSetup [
     {targetBone {boneName=net_base_2}}
     {targetBone {boneName=net_base_3}}
     {targetBone {boneName=net_base_1}}
     {targetBone {boneName=Melee_TwoHanded_Handle}}
    ]
    constraintWeights [1,1,1,1]
   }
   n6 = "Constraint - net_tip_5_base (PPPP)"!(
    ComponentPose=@n5
    Alpha=1
    ConstraintWeights_0=0.6642
    ConstraintWeights_1:float=0.285
    ConstraintWeights_2:float=0.0505
    ConstraintWeights_3:float=0.0004
   ) -> (Pose) {
    alphaBoolBlend {blendOption=Linear}
    boneToModify {boneName=net_tip_5_base}
    constraintSetup [
     {targetBone {boneName=net_base_2}}
     {targetBone {boneName=net_base_3}}
     {targetBone {boneName=net_base_1}}
     {targetBone {boneName=Melee_TwoHanded_Handle}}
    ]
    constraintWeights [1,1,1,1]
   }
   n7 = "Constraint - net_tip_6_base (PPPP)"!(
    ComponentPose=@n6
    Alpha=1
    ConstraintWeights_0=0.6545
    ConstraintWeights_1:float=0.2995
    ConstraintWeights_2:float=0.0456
    ConstraintWeights_3:float=0.0004
   ) -> (Pose) {
    alphaBoolBlend {blendOption=Linear}
    boneToModify {boneName=net_tip_6_base}
    constraintSetup [
     {targetBone {boneName=net_base_2}}
     {targetBone {boneName=net_base_3}}
     {targetBone {boneName=net_base_1}}
     {targetBone {boneName=Melee_TwoHanded_Handle}}
    ]
    constraintWeights [1,1,1,1]
   }
   n8 = "Constraint - net_tip_7_base (PPPP)"!(
    ComponentPose=@n7
    Alpha=1
    ConstraintWeights_0=0.6366
    ConstraintWeights_1:float=0.3266
    ConstraintWeights_2:float=0.0366
    ConstraintWeights_3:float=0.0002
   ) -> (Pose) {
    alphaBoolBlend {blendOption=Linear}
    boneToModify {boneName=net_tip_7_base}
    constraintSetup [
     {targetBone {boneName=net_base_2}}
     {targetBone {boneName=net_base_3}}
     {targetBone {boneName=net_base_1}}
     {targetBone {boneName=Melee_TwoHanded_Handle}}
    ]
    constraintWeights [1,1,1,1]
   }
   n9 = "Constraint - net_tip_8_base (PPPP)"!(
    ComponentPose=@n8
    Alpha=1
    ConstraintWeights_0=0.6534
    ConstraintWeights_1:float=0.3013
    ConstraintWeights_2:float=0.045
    ConstraintWeights_3:float=0.0004
   ) -> (Pose) {
    alphaBoolBlend {blendOption=Linear}
    boneToModify {boneName=net_tip_8_base}
    constraintSetup [
     {targetBone {boneName=net_base_2}}
     {targetBone {boneName=net_base_3}}
     {targetBone {boneName=net_base_1}}
     {targetBone {boneName=Melee_TwoHanded_Handle}}
    ]
    constraintWeights [1,1,1,1]
   }
   n10 = "Constraint - net_tip_9_base (PPPP)"!(
    ComponentPose=@n9
    Alpha=1
    ConstraintWeights_0=0.6489
    ConstraintWeights_1:float=0.308
    ConstraintWeights_2:float=0.0428
    ConstraintWeights_3:float=0.0003
   ) -> (Pose) {
    alphaBoolBlend {blendOption=Linear}
    boneToModify {boneName=net_tip_9_base}
    constraintSetup [
     {targetBone {boneName=net_base_2}}
     {targetBone {boneName=net_base_3}}
     {targetBone {boneName=net_base_1}}
     {targetBone {boneName=Melee_TwoHanded_Handle}}
    ]
    constraintWeights [1,1,1,1]
   }
   n11 = "Constraint - net_tip_10_base (PPPP)"!(
    ComponentPose=@n10
    Alpha=1
    ConstraintWeights_0=0.6969
    ConstraintWeights_1:float=0.2186
    ConstraintWeights_2:float=0.0834
    ConstraintWeights_3:float=0.0011
   ) -> (Pose) {
    alphaBoolBlend {blendOption=Linear}
    boneToModify {boneName=net_tip_10_base}
    constraintSetup [
     {targetBone {boneName=net_base_2}}
     {targetBone {boneName=net_base_3}}
     {targetBone {boneName=net_base_1}}
     {targetBone {boneName=Melee_TwoHanded_Handle}}
    ]
    constraintWeights [1,1,1,1]
   }
   n12 = "Constraint - net_tip_11_base (PPPP)"!(
    ComponentPose=@n11
    Alpha=1
    ConstraintWeights_0=0.6606
    ConstraintWeights_1:float=0.2902
    ConstraintWeights_2:float=0.0487
    ConstraintWeights_3:float=0.0004
   ) -> (Pose) {
    alphaBoolBlend {blendOption=Linear}
    boneToModify {boneName=net_tip_11_base}
    constraintSetup [
     {targetBone {boneName=net_base_2}}
     {targetBone {boneName=net_base_3}}
     {targetBone {boneName=net_base_1}}
     {targetBone {boneName=Melee_TwoHanded_Handle}}
    ]
    constraintWeights [1,1,1,1]
   }
   n13 = "Constraint - net_tip_12_base (PPPP)"!(
    ComponentPose=@n12
    Alpha=1
    ConstraintWeights_0=0.6955
    ConstraintWeights_1:float=0.2212
    ConstraintWeights_2:float=0.0823
    ConstraintWeights_3:float=0.001
   ) -> (Pose) {
    alphaBoolBlend {blendOption=Linear}
    boneToModify {boneName=net_tip_12_base}
    constraintSetup [
     {targetBone {boneName=net_base_2}}
     {targetBone {boneName=net_base_3}}
     {targetBone {boneName=net_base_1}}
     {targetBone {boneName=Melee_TwoHanded_Handle}}
    ]
    constraintWeights [1,1,1,1]
   }
   n14 = "Constraint - net_tip_13_base (PPPP)"!(
    ComponentPose=@n13
    Alpha=1
    ConstraintWeights_0=0.6343
    ConstraintWeights_1:float=0.3297
    ConstraintWeights_2:float=0.0357
    ConstraintWeights_3:float=0.0002
   ) -> (Pose) {
    alphaBoolBlend {blendOption=Linear}
    boneToModify {boneName=net_tip_13_base}
    constraintSetup [
     {targetBone {boneName=net_base_2}}
     {targetBone {boneName=net_base_3}}
     {targetBone {boneName=net_base_1}}
     {targetBone {boneName=Melee_TwoHanded_Handle}}
    ]
    constraintWeights [1,1,1,1]
   }
   n15 = "Constraint - net_tip_14_base (PPPP)"!(
    ComponentPose=@n14
    Alpha=1
    ConstraintWeights_0=0.7047
    ConstraintWeights_1:float=0.1989
    ConstraintWeights_2:float=0.0951
    ConstraintWeights_3:float=0.0013
   ) -> (Pose) {
    alphaBoolBlend {blendOption=Linear}
    boneToModify {boneName=net_tip_14_base}
    constraintSetup [
     {targetBone {boneName=net_base_2}}
     {targetBone {boneName=net_base_3}}
     {targetBone {boneName=net_base_1}}
     {targetBone {boneName=Melee_TwoHanded_Handle}}
    ]
    constraintWeights [1,1,1,1]
   }
   n16 = "Constraint - net_tip_15_base (PPPP)"!(
    ComponentPose=@n15
    Alpha=1
    ConstraintWeights_0=0.6622
    ConstraintWeights_1:float=0.2879
    ConstraintWeights_2:float=0.0495
    ConstraintWeights_3:float=0.0004
   ) -> (Pose) {
    alphaBoolBlend {blendOption=Linear}
    boneToModify {boneName=net_tip_15_base}
    constraintSetup [
     {targetBone {boneName=net_base_2}}
     {targetBone {boneName=net_base_3}}
     {targetBone {boneName=net_base_1}}
     {targetBone {boneName=Melee_TwoHanded_Handle}}
    ]
    constraintWeights [1,1,1,1]
   }
   n17 = "Constraint - net_tip_16_base (PPPP)"!(
    ComponentPose=@n16
    Alpha=1
    ConstraintWeights_0=0.6551
    ConstraintWeights_1:float=0.2986
    ConstraintWeights_2:float=0.0459
    ConstraintWeights_3:float=0.0004
   ) -> (Pose) {
    alphaBoolBlend {blendOption=Linear}
    boneToModify {boneName=net_tip_16_base}
    constraintSetup [
     {targetBone {boneName=net_base_2}}
     {targetBone {boneName=net_base_3}}
     {targetBone {boneName=net_base_1}}
     {targetBone {boneName=Melee_TwoHanded_Handle}}
    ]
    constraintWeights [1,1,1,1]
   }
   n18 = Anim Dynamics - Bone: net_tip_1(ComponentPose=@n17, Alpha=1, ExternalForce="0,0,0") {
    angularDampingOverride=1
    angularSpringConstant=20000
    bAngularSpring=true
    bOverrideAngularDamping=true
    bUseGravityOverride=true
    gravityScale=0.35
    simulationSpace=RootRelative
    alphaBoolBlend {blendOption=Linear}
    boundBone {boneName=net_tip_1}
    componentLinearAccScale=<1 1 1>
    componentLinearVelScale=<1 1 1>
    gravityOverride {z=-980}
    physicsBodyDefinitions [{
      boundBone {boneName=net_tip_1}
      boxExtents=<20 30 30>
      constraintSetup {angularLimitsMax=<0 15 15>, angularLimitsMin=<0 -15 -15>, angularTarget=<1 0 0>}
      localJointOffset=<4 0 0>
     }]
   }
   n19 = Anim Dynamics - Bone: net_tip_2(ComponentPose=@n18, Alpha=1, ExternalForce="0,0,0") {
    angularDampingOverride=1
    angularSpringConstant=20000
    bAngularSpring=true
    bOverrideAngularDamping=true
    bUseGravityOverride=true
    gravityScale=0.35
    simulationSpace=RootRelative
    alphaBoolBlend {blendOption=Linear}
    boundBone {boneName=net_tip_2}
    componentLinearAccScale=<1 1 1>
    componentLinearVelScale=<1 1 1>
    gravityOverride {z=-980}
    physicsBodyDefinitions [{
      boundBone {boneName=net_tip_2}
      boxExtents=<20 30 30>
      constraintSetup {angularLimitsMax=<0 15 15>, angularLimitsMin=<0 -15 -15>, angularTarget=<1 0 0>}
      localJointOffset=<4 0 0>
     }]
   }
   n20 = Anim Dynamics - Bone: net_tip_3(ComponentPose=@n19, Alpha=1, ExternalForce="0,0,0") {
    angularDampingOverride=1
    angularSpringConstant=20000
    bAngularSpring=true
    bOverrideAngularDamping=true
    bUseGravityOverride=true
    gravityScale=0.35
    simulationSpace=RootRelative
    alphaBoolBlend {blendOption=Linear}
    boundBone {boneName=net_tip_3}
    componentLinearAccScale=<1 1 1>
    componentLinearVelScale=<1 1 1>
    gravityOverride {z=-980}
    physicsBodyDefinitions [{
      boundBone {boneName=net_tip_3}
      boxExtents=<20 30 30>
      constraintSetup {angularLimitsMax=<0 15 15>, angularLimitsMin=<0 -15 -15>, angularTarget=<1 0 0>}
      localJointOffset=<4 0 0>
     }]
   }
   n21 = Anim Dynamics - Bone: net_tip_4(ComponentPose=@n20, Alpha=1, ExternalForce="0,0,0") {
    angularDampingOverride=1
    angularSpringConstant=20000
    bAngularSpring=true
    bOverrideAngularDamping=true
    bUseGravityOverride=true
    gravityScale=0.35
    simulationSpace=RootRelative
    alphaBoolBlend {blendOption=Linear}
    boundBone {boneName=net_tip_4}
    componentLinearAccScale=<1 1 1>
    componentLinearVelScale=<1 1 1>
    gravityOverride {z=-980}
    physicsBodyDefinitions [{
      boundBone {boneName=net_tip_4}
      boxExtents=<20 30 30>
      constraintSetup {angularLimitsMax=<0 15 15>, angularLimitsMin=<0 -15 -15>, angularTarget=<1 0 0>}
      localJointOffset=<4 0 0>
     }]
   }
   n22 = Anim Dynamics - Bone: net_tip_5(ComponentPose=@n21, Alpha=1, ExternalForce="0,0,0") {
    angularDampingOverride=1
    angularSpringConstant=20000
    bAngularSpring=true
    bOverrideAngularDamping=true
    bUseGravityOverride=true
    gravityScale=0.35
    simulationSpace=RootRelative
    alphaBoolBlend {blendOption=Linear}
    boundBone {boneName=net_tip_5}
    componentLinearAccScale=<1 1 1>
    componentLinearVelScale=<1 1 1>
    gravityOverride {z=-980}
    physicsBodyDefinitions [{
      boundBone {boneName=net_tip_5}
      boxExtents=<20 30 30>
      constraintSetup {angularLimitsMax=<0 15 15>, angularLimitsMin=<0 -15 -15>, angularTarget=<1 0 0>}
      localJointOffset=<4 0 0>
     }]
   }
   n23 = Anim Dynamics - Bone: net_tip_6(ComponentPose=@n22, Alpha=1, ExternalForce="0,0,0") {
    angularDampingOverride=1
    angularSpringConstant=20000
    bAngularSpring=true
    bOverrideAngularDamping=true
    bUseGravityOverride=true
    gravityScale=0.35
    simulationSpace=RootRelative
    alphaBoolBlend {blendOption=Linear}
    boundBone {boneName=net_tip_6}
    componentLinearAccScale=<1 1 1>
    componentLinearVelScale=<1 1 1>
    gravityOverride {z=-980}
    physicsBodyDefinitions [{
      boundBone {boneName=net_tip_6}
      boxExtents=<20 30 30>
      constraintSetup {angularLimitsMax=<0 15 15>, angularLimitsMin=<0 -15 -15>, angularTarget=<1 0 0>}
      localJointOffset=<4 0 0>
     }]
   }
   n24 = Anim Dynamics - Bone: net_tip_7(ComponentPose=@n23, Alpha=1, ExternalForce="0,0,0") {
    angularDampingOverride=1
    angularSpringConstant=20000
    bAngularSpring=true
    bOverrideAngularDamping=true
    bUseGravityOverride=true
    gravityScale=0.35
    simulationSpace=RootRelative
    alphaBoolBlend {blendOption=Linear}
    boundBone {boneName=net_tip_7}
    componentLinearAccScale=<1 1 1>
    componentLinearVelScale=<1 1 1>
    gravityOverride {z=-980}
    physicsBodyDefinitions [{
      boundBone {boneName=net_tip_7}
      boxExtents=<20 30 30>
      constraintSetup {angularLimitsMax=<0 15 15>, angularLimitsMin=<0 -15 -15>, angularTarget=<1 0 0>}
      localJointOffset=<4 0 0>
     }]
   }
   n25 = Anim Dynamics - Bone: net_tip_8(ComponentPose=@n24, Alpha=1, ExternalForce="0,0,0") {
    angularDampingOverride=1
    angularSpringConstant=20000
    bAngularSpring=true
    bOverrideAngularDamping=true
    bUseGravityOverride=true
    gravityScale=0.35
    simulationSpace=RootRelative
    alphaBoolBlend {blendOption=Linear}
    boundBone {boneName=net_tip_8}
    componentLinearAccScale=<1 1 1>
    componentLinearVelScale=<1 1 1>
    gravityOverride {z=-980}
    physicsBodyDefinitions [{
      boundBone {boneName=net_tip_8}
      boxExtents=<20 30 30>
      constraintSetup {angularLimitsMax=<0 15 15>, angularLimitsMin=<0 -15 -15>, angularTarget=<1 0 0>}
      localJointOffset=<4 0 0>
     }]
   }
   n26 = Anim Dynamics - Bone: net_tip_9(ComponentPose=@n25, Alpha=1, ExternalForce="0,0,0") {
    angularDampingOverride=1
    angularSpringConstant=20000
    bAngularSpring=true
    bOverrideAngularDamping=true
    bUseGravityOverride=true
    gravityScale=0.35
    simulationSpace=RootRelative
    alphaBoolBlend {blendOption=Linear}
    boundBone {boneName=net_tip_9}
    componentLinearAccScale=<1 1 1>
    componentLinearVelScale=<1 1 1>
    gravityOverride {z=-980}
    physicsBodyDefinitions [{
      boundBone {boneName=net_tip_9}
      boxExtents=<20 30 30>
      constraintSetup {angularLimitsMax=<0 15 15>, angularLimitsMin=<0 -15 -15>, angularTarget=<1 0 0>}
      localJointOffset=<4 0 0>
     }]
   }
   n27 = Anim Dynamics - Bone: net_tip_10(ComponentPose=@n26, Alpha=1, ExternalForce="0,0,0") {
    angularDampingOverride=1
    angularSpringConstant=20000
    bAngularSpring=true
    bOverrideAngularDamping=true
    bUseGravityOverride=true
    gravityScale=0.35
    simulationSpace=RootRelative
    alphaBoolBlend {blendOption=Linear}
    boundBone {boneName=net_tip_10}
    componentLinearAccScale=<1 1 1>
    componentLinearVelScale=<1 1 1>
    gravityOverride {z=-980}
    physicsBodyDefinitions [{
      boundBone {boneName=net_tip_10}
      boxExtents=<20 30 30>
      constraintSetup {angularLimitsMax=<0 15 15>, angularLimitsMin=<0 -15 -15>, angularTarget=<1 0 0>}
      localJointOffset=<4 0 0>
     }]
   }
   n28 = Anim Dynamics - Bone: net_tip_11(ComponentPose=@n27, Alpha=1, ExternalForce="0,0,0") {
    angularDampingOverride=1
    angularSpringConstant=20000
    bAngularSpring=true
    bOverrideAngularDamping=true
    bUseGravityOverride=true
    gravityScale=0.35
    simulationSpace=RootRelative
    alphaBoolBlend {blendOption=Linear}
    boundBone {boneName=net_tip_11}
    componentLinearAccScale=<1 1 1>
    componentLinearVelScale=<1 1 1>
    gravityOverride {z=-980}
    physicsBodyDefinitions [{
      boundBone {boneName=net_tip_11}
      boxExtents=<20 30 30>
      constraintSetup {angularLimitsMax=<0 15 15>, angularLimitsMin=<0 -15 -15>, angularTarget=<1 0 0>}
      localJointOffset=<4 0 0>
     }]
   }
   n29 = Anim Dynamics - Bone: net_tip_12(ComponentPose=@n28, Alpha=1, ExternalForce="0,0,0") {
    angularDampingOverride=1
    angularSpringConstant=20000
    bAngularSpring=true
    bOverrideAngularDamping=true
    bUseGravityOverride=true
    gravityScale=0.35
    simulationSpace=RootRelative
    alphaBoolBlend {blendOption=Linear}
    boundBone {boneName=net_tip_12}
    componentLinearAccScale=<1 1 1>
    componentLinearVelScale=<1 1 1>
    gravityOverride {z=-980}
    physicsBodyDefinitions [{
      boundBone {boneName=net_tip_12}
      boxExtents=<20 30 30>
      constraintSetup {angularLimitsMax=<0 15 15>, angularLimitsMin=<0 -15 -15>, angularTarget=<1 0 0>}
      localJointOffset=<4 0 0>
     }]
   }
   n30 = Anim Dynamics - Bone: net_tip_13(ComponentPose=@n29, Alpha=1, ExternalForce="0,0,0") {
    angularDampingOverride=1
    angularSpringConstant=20000
    bAngularSpring=true
    bOverrideAngularDamping=true
    bUseGravityOverride=true
    gravityScale=0.35
    simulationSpace=RootRelative
    alphaBoolBlend {blendOption=Linear}
    boundBone {boneName=net_tip_13}
    componentLinearAccScale=<1 1 1>
    componentLinearVelScale=<1 1 1>
    gravityOverride {z=-980}
    physicsBodyDefinitions [{
      boundBone {boneName=net_tip_13}
      boxExtents=<20 30 30>
      constraintSetup {angularLimitsMax=<0 15 15>, angularLimitsMin=<0 -15 -15>, angularTarget=<1 0 0>}
      localJointOffset=<4 0 0>
     }]
   }
   n31 = Anim Dynamics - Bone: net_tip_14(ComponentPose=@n30, Alpha=1, ExternalForce="0,0,0") {
    angularDampingOverride=1
    angularSpringConstant=20000
    bAngularSpring=true
    bOverrideAngularDamping=true
    bUseGravityOverride=true
    gravityScale=0.35
    simulationSpace=RootRelative
    alphaBoolBlend {blendOption=Linear}
    boundBone {boneName=net_tip_14}
    componentLinearAccScale=<1 1 1>
    componentLinearVelScale=<1 1 1>
    gravityOverride {z=-980}
    physicsBodyDefinitions [{
      boundBone {boneName=net_tip_14}
      boxExtents=<20 30 30>
      constraintSetup {angularLimitsMax=<0 15 15>, angularLimitsMin=<0 -15 -15>, angularTarget=<1 0 0>}
      localJointOffset=<4 0 0>
     }]
   }
   n32 = Anim Dynamics - Bone: net_tip_15(ComponentPose=@n31, Alpha=1, ExternalForce="0,0,0") {
    angularDampingOverride=1
    angularSpringConstant=20000
    bAngularSpring=true
    bOverrideAngularDamping=true
    bUseGravityOverride=true
    gravityScale=0.35
    simulationSpace=RootRelative
    alphaBoolBlend {blendOption=Linear}
    boundBone {boneName=net_tip_15}
    componentLinearAccScale=<1 1 1>
    componentLinearVelScale=<1 1 1>
    gravityOverride {z=-980}
    physicsBodyDefinitions [{
      boundBone {boneName=net_tip_15}
      boxExtents=<20 30 30>
      constraintSetup {angularLimitsMax=<0 15 15>, angularLimitsMin=<0 -15 -15>, angularTarget=<1 0 0>}
      localJointOffset=<4 0 0>
     }]
   }
   n33 = Anim Dynamics - Bone: net_tip_16(ComponentPose=@n32, Alpha=1, ExternalForce="0,0,0") {
    angularDampingOverride=1
    angularSpringConstant=20000
    bAngularSpring=true
    bOverrideAngularDamping=true
    bUseGravityOverride=true
    gravityScale=0.35
    simulationSpace=RootRelative
    alphaBoolBlend {blendOption=Linear}
    boundBone {boneName=net_tip_16}
    componentLinearAccScale=<1 1 1>
    componentLinearVelScale=<1 1 1>
    gravityOverride {z=-980}
    physicsBodyDefinitions [{
      boundBone {boneName=net_tip_16}
      boxExtents=<20 30 30>
      constraintSetup {angularLimitsMax=<0 15 15>, angularLimitsMin=<0 -15 -15>, angularTarget=<1 0 0>}
      localJointOffset=<4 0 0>
     }]
   }
   Output Pose(Result=Component To Local(ComponentPose=@n33)) {layerGroup=None}
  ]}
 EventGraph {nodes [
   n1 = Event BlueprintUpdateAnimation()
   > n2 = Gravity Override State Function DEPRECATED(CharacterPartAnimInstance=Self-Reference(), GameTimeInSeconds=@n1.DeltaTimeX, GravityOverrideParams=Get Gravity Override Params())
   > Set Gravity Override(Gravity Override=@n2)
   TryGetPawnOwner()
  ]}
]
variables [
 Gravity Override {type=Vector}
 Gravity Override Params {type=GravityOverrideParamsStruct}
]
